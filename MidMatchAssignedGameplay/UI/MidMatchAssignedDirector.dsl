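; A Dynamic UI director for MidMatchAssignedGameplay that registers the feature's UI scene for the owning local player.
; On BeginPlay it queries that player's Fort Global UI Context and adds MidMatchAssignedScene only if the player is in a zone.
; The check runs once, with no tick logic or scene removal in this asset.
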(event EventBeginPlay
  (if (GlobalUIContext|IsinZone (PlayerController|LocalPlayerSubsystems|GetFortGlobalUIContext (DynamicUIDirector|GetOwningLocalPlayerController)))
    (DynamicUIDirector|AddScene "/MidMatchAssignedGameplay/UI/MidMatchAssignedScene.MidMatchAssignedScene")))

(event EventTick
  (params (DeltaSeconds :type Float%20%28single-precision%29)))

(fn ConstructionScript)
